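# Sound cue for the shotgun's empty sound, routed to the weapon foley sound class.
# Nested mixers layer an assault rifle reload element, a pistol foley click and a negative UI tone, each weighted separately.
# The two weapon samples pass through modulators with fixed upward pitch shifts, so there is no random variation.
# A concurrency override blocks new instances at the limit instead of cutting off playing ones.
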SoundCue /Game/Sounds/Fort_Weapons/Shotgun/Shotgun_Empty_Cue
root=Mixer_0
nodes [
 {id=Mixer_0, children [Modulator_8,Mixer_3], properties {inputVolume [0.5,1]}}
 {id=Mixer_3, children [Modulator_2, WavePlayer_13], properties {inputVolume [1,0.4]}}
 {id=Modulator_2, children [WavePlayer_14], properties {pitchMax=1.3, pitchMin=1.3}}
 {id=Modulator_8, children [WavePlayer_22], properties {pitchMax=2, pitchMin=2}}
 {id=WavePlayer_13, properties {soundWaveAssetPtr=/Game/Sounds/Fort_UI_Sounds/Gameplay/fort_ui_new_neg_01}}
 {id=WavePlayer_14, properties {soundWaveAssetPtr=/Game/Sounds/Fort_Weapons/pistol/pistol_foley/auto_sdk_end_01}}
 {id=WavePlayer_22, properties {soundWaveAssetPtr=/Game/Sounds/Fort_Weapons/AutoRifle/Reload_Elements/assault_end_sdk_01}}
]
properties {
 bOverrideConcurrency=true
 maxDistance=2097152
 soundClassObject=&Fort_WeaponFoley
 volumeMultiplier=0.501
 attenuationOverrides {falloffDistance=1600}
 concurrencyOverrides {resolutionRule=PreventNew}
}
