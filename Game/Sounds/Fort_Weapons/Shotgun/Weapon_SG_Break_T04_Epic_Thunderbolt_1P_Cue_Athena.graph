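# First-person fire sound cue for the epic Thunderbolt break shotgun in Athena, using first-person gunfire concurrency and attenuation.
# Each shot layers a randomly picked, slightly pitched-down "bigger" shotgun wave with a briefly delayed, more deeply pitched close break-shotgun wave, then modulates the mix.
# A quality switch gives the first level three random main waves, while the other three levels share a duplicate branch with two.
# Both outer mixers carry unconnected inputs.
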SoundCue /Game/Sounds/Fort_Weapons/Shotgun/Weapon_SG_Break_T04_Epic_Thunderbolt_1P_Cue_Athena
root=QualityLevel_0
nodes [
 {id=Delay_1, children [Modulator_5], properties {delayMax=0.02, delayMin=0.02}}
 {id=Delay_2, children [Modulator_9], properties {delayMax=0.02, delayMin=0.02}}
 {id=Mixer_2, children [Modulator_4,Delay_1], properties {inputVolume [1,1]}}
 {id=Mixer_3, children [Modulator_7,null,null], properties {inputVolume [1,0.8,0.8]}}
 {id=Mixer_4, children [Modulator_8,Delay_2], properties {inputVolume [1,1]}}
 {id=Mixer_5, children [Modulator_11,null,null], properties {inputVolume [1,1,1]}}
 {id=Modulator_11, children [Mixer_4]}
 {id=Modulator_4, children [Random_3], properties {pitchMax=0.95, pitchMin=0.9}}
 {id=Modulator_5, children [WavePlayer_15], properties {pitchMax=0.8, pitchMin=0.5}}
 {id=Modulator_7, children [Mixer_2]}
 {id=Modulator_8, children [Random_5], properties {pitchMax=0.95, pitchMin=0.9}}
 {id=Modulator_9, children [WavePlayer_23], properties {pitchMax=0.8, pitchMin=0.5}}
 {id=QualityLevel_0, children [Mixer_3,Mixer_5,Mixer_5,Mixer_5]}
 {id=Random_3, children [WavePlayer_12, WavePlayer_13, WavePlayer_14], properties {weights [1,1,1]}}
 {id=Random_5, children [WavePlayer_20, WavePlayer_22], properties {weights [1,1]}}
 {id=WavePlayer_12, properties {soundWaveAssetPtr=&/weapons_shotguns_bigger_013}}
 {id=WavePlayer_13, properties {soundWaveAssetPtr=&/weapons_shotguns_bigger_014}}
 {id=WavePlayer_14, properties {soundWaveAssetPtr=&/weapons_shotguns_bigger_015}}
 {id=WavePlayer_15, properties {soundWaveAssetPtr=&/weapon_shotgun_break2_close_02}}
 {id=WavePlayer_20, properties {soundWaveAssetPtr=&weapons_shotguns_bigger_013}}
 {id=WavePlayer_22, properties {soundWaveAssetPtr=&weapons_shotguns_bigger_015}}
 {id=WavePlayer_23, properties {soundWaveAssetPtr=&weapon_shotgun_break2_close_02}}
]
properties {
 attenuationSettings=&Fort_Weapon_Projectile_FirstPerson_Athena_Att
 maxDistance=2097152
 soundClassObject=&Fort_WeaponRanged_Fire_1P_PSM
 volumeMultiplier=0.7
 attenuationOverrides {falloffDistance=5000, attenuationShapeExtents {x=1000}}
 concurrencySet [&Weapon_Gunfire_FirstPerson_Conc]
}
