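# Sound cue for the Crusader basic shotgun's firing, using third-person and distant weapon-fire sound classes with gunfire concurrency limits.
# Each mix pairs a close layer with a quieter, randomized distant shot, each with its own attenuation and class.
# The close layer combines a random pitched-up blast, a slightly delayed tail and later, layered trigger sounds.
# A quality switch holds two near-identical mixes, the second with fewer distant variants, and two empty slots.
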SoundCue /Game/Sounds/Fort_Weapons/Shotgun/Weapon_SG_Standard_T01_Basic_Crusader_Fire_Cue
root=QualityLevel_0
nodes [
 {id=Attenuation_2, children [Modulator_9], properties {attenuationSettings=/Game/Sounds/Attenuation/Fort_Weapon_Shotgun_Sml_Att}}
 {id=Attenuation_3, children [Modulator_7], properties {attenuationSettings=/Game/Sounds/Attenuation/Fort_Weapon_Projectile_Distant_StW_Att}}
 {id=Attenuation_4, children [Modulator_14], properties {attenuationSettings=&Fort_Weapon_Shotgun_Sml_Att}}
 {id=Attenuation_5, children [Modulator_12], properties {attenuationSettings=&Fort_Weapon_Projectile_Distant_StW_Att}}
 {id=Delay_2, children [Modulator_6], properties {delayMax=0.015, delayMin=0.01}}
 {id=Delay_3, children [Modulator_8], properties {delayMax=0.25, delayMin=0.2}}
 {id=Delay_4, children [Modulator_11], properties {delayMax=0.015, delayMin=0.01}}
 {id=Delay_5, children [Modulator_13], properties {delayMax=0.25, delayMin=0.2}}
 {id=Mixer_3, children [Modulator_5,Delay_2,Delay_3], properties {inputVolume [1,0.7,1]}}
 {id=Mixer_4, children [WavePlayer_16, WavePlayer_17], properties {inputVolume [1,1]}}
 {id=Mixer_5, children [SoundClass_2,SoundClass_3], properties {inputVolume [1,0.7]}}
 {id=Mixer_6, children [Modulator_10,Delay_4,Delay_5], properties {inputVolume [1,0.7,1]}}
 {id=Mixer_7, children [WavePlayer_25, WavePlayer_26], properties {inputVolume [1,1]}}
 {id=Mixer_8, children [SoundClass_4,SoundClass_5], properties {inputVolume [1,0.7]}}
 {id=Modulator_10, children [Random_5], properties {pitchMax=1.1, pitchMin=1.1}}
 {id=Modulator_11, children [WavePlayer_18]}
 {id=Modulator_12, children [Random_4], properties {pitchMax=1.1}}
 {id=Modulator_13, children [Mixer_7]}
 {id=Modulator_14, children [Mixer_6]}
 {id=Modulator_5, children [Random_3], properties {pitchMax=1.1, pitchMin=1.1}}
 {id=Modulator_6, children [WavePlayer_9]}
 {id=Modulator_7, children [Random_2], properties {pitchMax=1.1}}
 {id=Modulator_8, children [Mixer_4]}
 {id=Modulator_9, children [Mixer_3]}
 {id=QualityLevel_0, children [Mixer_5,Mixer_8,null,null]}
 {id=Random_2, children [WavePlayer_10, WavePlayer_11, WavePlayer_12, WavePlayer_13], properties {weights [1,1,1,1]}}
 {id=Random_3, children [WavePlayer_14, WavePlayer_15], properties {weights [1,1]}}
 {id=Random_4, children [WavePlayer_19, WavePlayer_20, WavePlayer_22], properties {weights [1,1,1]}}
 {id=Random_5, children [WavePlayer_23, WavePlayer_24], properties {weights [1,1]}}
 {
  id=SoundClass_2
  children [Attenuation_2]
  properties {soundClassOverride=&Fort_WeaponRanged_Fire_3P_PSM}
 }
 {
  id=SoundClass_3
  children [Attenuation_3]
  properties {soundClassOverride=&Fort_WeaponRanged_Fire_Distant}
 }
 {
  id=SoundClass_4
  children [Attenuation_4]
  properties {soundClassOverride=&Fort_WeaponRanged_Fire_3P_PSM}
 }
 {
  id=SoundClass_5
  children [Attenuation_5]
  properties {soundClassOverride=&Fort_WeaponRanged_Fire_Distant}
 }
 {id=WavePlayer_10, properties {soundWaveAssetPtr=/Game/Sounds/Fort_Weapons/Distant/Distant_low_shot_01c}}
 {id=WavePlayer_11, properties {soundWaveAssetPtr=/Game/Sounds/Fort_Weapons/Distant/Distant_low_shot_01d}}
 {id=WavePlayer_12, properties {soundWaveAssetPtr=/Game/Sounds/Fort_Weapons/Distant/Distant_low_shot_01b}}
 {id=WavePlayer_13, properties {soundWaveAssetPtr=/Game/Sounds/Fort_Weapons/Distant/Distant_low_shot_01a}}
 {id=WavePlayer_14, properties {soundWaveAssetPtr=&/weapons_shotguns_bigger_08}}
 {id=WavePlayer_15, properties {soundWaveAssetPtr=&/weapons_shotguns_bigger_018}}
 {id=WavePlayer_16, properties {soundWaveAssetPtr=&/Shotgun_Trigger_Basic}}
 {id=WavePlayer_17, properties {soundWaveAssetPtr=&/Shotgun_Trigger_Basic03}}
 {id=WavePlayer_18, properties {soundWaveAssetPtr=&/Shotgun_Fire_Tail_Stereo_03}}
 {id=WavePlayer_19, properties {soundWaveAssetPtr=&Distant_low_shot_01c}}
 {id=WavePlayer_20, properties {soundWaveAssetPtr=&Distant_low_shot_01d}}
 {id=WavePlayer_22, properties {soundWaveAssetPtr=&Distant_low_shot_01a}}
 {id=WavePlayer_23, properties {soundWaveAssetPtr=&weapons_shotguns_bigger_08}}
 {id=WavePlayer_24, properties {soundWaveAssetPtr=&weapons_shotguns_bigger_018}}
 {id=WavePlayer_25, properties {soundWaveAssetPtr=&Shotgun_Trigger_Basic}}
 {id=WavePlayer_26, properties {soundWaveAssetPtr=&Shotgun_Trigger_Basic03}}
 {id=WavePlayer_9, properties {soundWaveAssetPtr=&Shotgun_Fire_Tail_Stereo_03}}
]
properties {
 maxDistance=20000
 soundClassObject=&Fort_WeaponRanged_Fire_3P_PSM
 volumeMultiplier=0.5
 attenuationOverrides {falloffDistance=5000, attenuationShapeExtents {x=1000}}
 concurrencySet [&Weapon_Gunfire_Local_Conc, &Weapon_Gunfire_Global_Conc]
}
