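# Sound cue for first-person firing of the mid-range shotgun in Athena.
# A Random node picks one of three equally weighted fire waves, is excluded from branch culling, and feeds a Modulator that varies playback with a raised pitch floor.
# It uses first-person projectile attenuation, a ranged-fire sound class and a shared first-person gunfire concurrency set.
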SoundCue /Game/Sounds/Fort_Weapons/Shotgun/Weapon_Shotgun_MidRange_Fire_1P_Athena_Cue
root=Modulator_0
nodes [
 {id=Modulator_0, children [Random_0], properties {pitchMin=1}}
 {id=Random_0, children [WavePlayer_0,WavePlayer_1,WavePlayer_8], properties {bShouldExcludeFromBranchCulling=true, weights [1,1,1]}}
 {id=WavePlayer_0, properties {soundWaveAssetPtr=&/Weapon_Shotgun_MidRange_Fire01}}
 {id=WavePlayer_1, properties {soundWaveAssetPtr=&/Weapon_Shotgun_MidRange_Fire02}}
 {id=WavePlayer_8, properties {soundWaveAssetPtr=&/Weapon_Shotgun_MidRange_Fire03}}
]
properties {
 attenuationSettings=&Fort_Weapon_Projectile_FirstPerson_Athena_Att
 maxDistance=2097152
 soundClassObject=&Fort_WeaponRanged_Fire_1P_PSM
 volumeMultiplier=0.8
 attenuationOverrides {falloffDistance=5000, attenuationShapeExtents {x=1000}}
 concurrencyOverrides {bLimitToOwner=true, maxCount=3, resolutionRule=StopOldest}
 concurrencySet [&Weapon_Gunfire_FirstPerson_Conc]
}
