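# Sound cue for the mid-range shotgun's fire sound in Athena, layered by distance into close, medium and distant gunshots.
# Each layer randomly picks a wave, applies a modulator and its own projectile attenuation, then feeds a weighted mixer that favors the close layer.
# The close layer uses the third-person fire sound class, while medium and distant share the distant fire class.
# Local and global gunfire concurrency sets are assigned.
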SoundCue /Game/Sounds/Fort_Weapons/Shotgun/Weapon_Shotgun_MidRange_Fire_Athena_Cue
root=Mixer_0
nodes [
 {id=Attenuation_0, children [Modulator_0], properties {attenuationSettings=&Fort_Weapon_Projectile_Close_Athena_Att}}
 {
  id=Attenuation_1
  children [Modulator_1]
  properties {attenuationSettings=&Fort_Weapon_Projectile_Medium_Mono_Athena_Att}
 }
 {
  id=Attenuation_2
  children [Modulator_2]
  properties {attenuationSettings=&Fort_Weapon_Projectile_Distant_Mono_Athena_Att}
 }
 {id=Mixer_0, children [SoundClass_0,SoundClass_6,SoundClass_1], properties {inputVolume [0.8,0.25,0.285]}}
 {id=Modulator_0, children [Random_0], properties {pitchMin=1}}
 {id=Modulator_1, children [Random_2], properties {volumeMax=1.95, volumeMin=1.9}}
 {id=Modulator_2, children [Random_1], properties {volumeMax=2.05, volumeMin=1.95}}
 {id=Random_0, children [WavePlayer_0,WavePlayer_1,WavePlayer_8], properties {bShouldExcludeFromBranchCulling=true, weights [1,1,1]}}
 {id=Random_1, children [WavePlayer_4,WavePlayer_5], properties {bShouldExcludeFromBranchCulling=true, weights [1,1]}}
 {id=Random_2, children [WavePlayer_6,WavePlayer_7], properties {bShouldExcludeFromBranchCulling=true, weights [1,1]}}
 {
  id=SoundClass_0
  children [Attenuation_0]
  properties {soundClassOverride=&Fort_WeaponRanged_Fire_3P_PSM}
 }
 {
  id=SoundClass_1
  children [Attenuation_2]
  properties {soundClassOverride=&Fort_WeaponRanged_Fire_Distant}
 }
 {
  id=SoundClass_6
  children [Attenuation_1]
  properties {soundClassOverride=&Fort_WeaponRanged_Fire_Distant}
 }
 {id=WavePlayer_0, properties {soundWaveAssetPtr=&/Weapon_Shotgun_MidRange_Fire01}}
 {id=WavePlayer_1, properties {soundWaveAssetPtr=&/Weapon_Shotgun_MidRange_Fire02}}
 {id=WavePlayer_4, properties {soundWaveAssetPtr=&/Weapon_Shotgun_MidRange_DistantNew_Fire01}}
 {id=WavePlayer_5, properties {soundWaveAssetPtr=&/Weapon_Shotgun_MidRange_DistantNew_Fire02}}
 {id=WavePlayer_6, properties {soundWaveAssetPtr=&/Weapon_Shotgun_MidRange_MediumNew_Fire01}}
 {id=WavePlayer_7, properties {soundWaveAssetPtr=&/Weapon_Shotgun_MidRange_MediumNew_Fire02}}
 {id=WavePlayer_8, properties {soundWaveAssetPtr=&/Weapon_Shotgun_MidRange_Fire03}}
]
properties {
 maxDistance=27500
 soundClassObject=&Fort_WeaponRanged_Fire_3P_PSM
 volumeMultiplier=1
 attenuationOverrides {falloffDistance=5000, attenuationShapeExtents {x=1000}}
 concurrencyOverrides {bLimitToOwner=true, maxCount=3, resolutionRule=StopOldest}
 concurrencySet [&Weapon_Gunfire_Local_Conc, &Weapon_Gunfire_Global_Conc]
}
