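# Sound cue for a shotgun pump-open foley sound, part of the weapon's reload elements.
# A single wave player feeds a modulator that boosts volume by a fixed amount, with the cue's own multiplier scaling it back to about half.
# It uses the weapon foley sound class and reload attenuation, and overrides concurrency so new instances are blocked rather than replacing playing ones at the limit.
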SoundCue /Game/Sounds/Fort_Weapons/Shotgun/Weapon_Shotgun_Open2_Cue
root=Modulator_0
nodes [
 {id=Modulator_0, children [WavePlayer_0], properties {volumeMax=2, volumeMin=2}}
 {id=WavePlayer_0, properties {soundWaveAssetPtr=/Game/Sounds/Fort_Weapons/Shotgun/Reload_Elements/weapon_shotgun_pump_openNew}}
]
properties {
 attenuationSettings=&Fort_Weapon_Reload_Att
 bOverrideConcurrency=true
 maxDistance=2250
 soundClassObject=&Fort_WeaponFoley
 volumeMultiplier=0.501
 attenuationOverrides {falloffDistance=1600}
 concurrencyOverrides {resolutionRule=PreventNew}
}
