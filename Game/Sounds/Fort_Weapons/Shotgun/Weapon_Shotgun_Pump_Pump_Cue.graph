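# Sound cue for the pump-action foley of the pump shotgun, in the weapon foley sound class.
# A single wave player feeds a modulator with fixed pitch and boosted volume, so playback has no random variation.
# It uses the shared weapon reload attenuation asset.
# Its concurrency override limits simultaneous plays per owner and stops the oldest when the limit is exceeded.
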SoundCue /Game/Sounds/Fort_Weapons/Shotgun/Weapon_Shotgun_Pump_Pump_Cue
root=Modulator_14
nodes [
 {id=Modulator_14, children [WavePlayer_1], properties {pitchMax=1, pitchMin=1, volumeMax=2, volumeMin=2}}
 {id=WavePlayer_1, properties {soundWaveAssetPtr=/Game/Sounds/Fort_Weapons/Shotgun/Reload_Elements/Pump_Shotgun_PumpNew}}
]
properties {
 attenuationSettings=&Fort_Weapon_Reload_Att
 bOverrideConcurrency=true
 maxDistance=2250
 soundClassObject=&Fort_WeaponFoley
 volumeMultiplier=0.351
 attenuationOverrides {falloffDistance=1600}
 concurrencyOverrides {bLimitToOwner=true, maxCount=2, resolutionRule=StopOldest}
}
