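# Sound cue for the tactical shotgun's cock, playing the pump foley from the shotgun reload elements.
# A single wave player feeds a modulator that slightly randomizes pitch on each play.
# It uses the weapon reload attenuation and foley sound class, with overridden concurrency that caps instances per owner and stops the oldest.
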SoundCue /Game/Sounds/Fort_Weapons/Shotgun/Weapon_Shotgun_Tactical_Cock_Cue
root=Modulator_14
nodes [
 {id=Modulator_14, children [WavePlayer_0], properties {pitchMax=1.1, pitchMin=0.9}}
 {id=WavePlayer_0, properties {soundWaveAssetPtr=/Game/Sounds/Fort_Weapons/Shotgun/Reload_Elements/Tactical_Pump}}
]
properties {
 attenuationSettings=&Fort_Weapon_Reload_Att
 bOverrideConcurrency=true
 maxDistance=2250
 soundClassObject=&Fort_WeaponFoley
 volumeMultiplier=0.702
 attenuationOverrides {falloffDistance=1600}
 concurrencyOverrides {bLimitToOwner=true, maxCount=2, resolutionRule=StopOldest}
}
