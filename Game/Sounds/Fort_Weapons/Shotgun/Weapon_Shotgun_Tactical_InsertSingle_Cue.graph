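# Sound cue for inserting a single shell during a tactical shotgun reload, classed as weapon foley.
# An equally weighted random node picks one of three shell-insert waves, which passes through a modulator for slight pitch variation and then an attenuation node at the root.
# The cue uses the shared weapon reload attenuation and overrides concurrency to prevent new instances once the limit is reached.
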SoundCue /Game/Sounds/Fort_Weapons/Shotgun/Weapon_Shotgun_Tactical_InsertSingle_Cue
root=Attenuation_13
nodes [
 {id=Attenuation_13, children [Modulator_13], properties {attenuationOverrides {bSpatialize=false}}}
 {id=Modulator_13, children [Random_0], properties {pitchMax=1.1, pitchMin=0.9, volumeMax=1, volumeMin=1}}
 {id=Random_0, children [WavePlayer_0,WavePlayer_1,WavePlayer_2], properties {weights [1,1,1]}}
 {id=WavePlayer_0, properties {soundWaveAssetPtr=/Game/Sounds/Fort_Weapons/Shotgun/Reload_Elements/Tactical_InsertShell_01}}
 {id=WavePlayer_1, properties {soundWaveAssetPtr=/Game/Sounds/Fort_Weapons/Shotgun/Reload_Elements/Tactical_InsertShell_03}}
 {id=WavePlayer_2, properties {soundWaveAssetPtr=/Game/Sounds/Fort_Weapons/Shotgun/Reload_Elements/Tactical_InsertShell_02}}
]
properties {
 attenuationSettings=&Fort_Weapon_Reload_Att
 bOverrideConcurrency=true
 maxDistance=2097152
 soundClassObject=&Fort_WeaponFoley
 volumeMultiplier=0.702
 attenuationOverrides {falloffDistance=1600}
 concurrencyOverrides {resolutionRule=PreventNew}
}
